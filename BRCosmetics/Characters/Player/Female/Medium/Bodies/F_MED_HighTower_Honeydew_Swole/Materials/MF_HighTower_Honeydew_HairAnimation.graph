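# Material function for the HighTower Honeydew character's hair, adding waving motion and a parameter-driven transition that reveals or hides strands and shifts their color.
# Time-driven sine waves add to World Position Offset, with vertex color R masking displacement, G blending speed and amplitude, and B offsetting phase.
# TransitionAmt drives a pre-skinned-space SphereMask into Opacity Mask, offset per strand by GMap green, plus a hue shift and additive tint on Base Color.
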MaterialFunction /BRCosmetics/Characters/Player/Female/Medium/Bodies/F_MED_HighTower_Honeydew_Swole/Materials/MF_HighTower_Honeydew_HairAnimation
graph {nodes [
  ComponentMask() {g=true, r=true}
  n1 = GetMaterialAttributes(None=FunctionInput() {inputName=Material, inputType=FunctionInput_MaterialAttributes}) {attributeGetTypes [#1, #2]}
  n2 = VertexColor()
  n3 = ScalarParameter() {defaultValue=1, group=HairAnimation, parameterName=H_Offset_Amt}
  n4 = Multiply(A=Time() {period=25}, B=ScalarParameter() {defaultValue=0.5, group=HairAnimation, parameterName=H_GlobalWaveSpeed})
  n5 = Add(A=@n4, B=ScalarParameter() {defaultValue=0.25, group=HairAnimation, parameterName=H_MinWaveSpeed})
  n6 = Add(A=@n4, B=ScalarParameter() {defaultValue=0.25, group=HairAnimation, parameterName=H_MaxWaveSpeed})
  n7 = Sine(None=Add(A=Multiply(A=@n2.B, B=@n3), B=LinearInterpolate(A=@n5, B=@n6, Alpha=@n2.G)))
  n8 = LinearInterpolate(
   A=ScalarParameter() {defaultValue=2, group=HairAnimation, parameterName=H_MinSineWave}
   B=ScalarParameter() {defaultValue=1, group=HairAnimation, parameterName=H_MaxSineWave}
   Alpha=@n2.G
  )
  n9 = Multiply(
   A=Multiply(A=@n2.R, B=Multiply(A=@n7, B=@n8))
   B=VectorParameter() {group=HairAnimation, parameterName=H_OffsetAmtAndDirection, defaultValue=<rgba 2 1 1 1>}.RGB
  )
  n10 = AppendVector(A=Constant(), B=ComponentMask(None=PreSkinnedLocalBounds().Max) {b=true, g=true})
  n11 = VertexInterpolator(VS=PreSkinnedPosition())
  n12 = Multiply(
   A=Add(
    A=FunctionInput(Preview=@n11.PS) {bUsePreviewValueAsDefault=true, inputName=PreskinnedPosition}
    B=VectorParameter() {group=HoneyDew, parameterName=PosOffset}.RGB
   )
   B=VectorParameter() {group=HoneyDew, parameterName=PosScale, defaultValue=<rgba 1 1 1 0>}.RGB
  )
  n13 = OneMinus(None=ScalarParameter() {defaultValue=1, group=HoneyDew, parameterName=TransitionAmt})
  n14 = ScalarParameter() {defaultValue=1, group=HoneyDew, parameterName=TransitionAmntPow}
  n15 = FunctionInput() {bUsePreviewValueAsDefault=true, inputName=GMap}
  n16 = Add(
   A=LinearInterpolate(
    A=ScalarParameter() {group=HoneyDew, parameterName=min}
    B=ScalarParameter() {group=HoneyDew, parameterName=Max}
    Alpha=Power(Base=@n13, Exp=@n14)
   ) {constA=35, constB=55}
   B=Multiply(A=ComponentMask(None=@n15) {g=true}, B=ScalarParameter() {group=HoneyDew, parameterName=StrandsGrowthOffset}) {constB=0}
  )
  n17 = ScalarParameter() {defaultValue=0.8, group=HoneyDew, parameterName=hard}
  n18 = Power(Base=@n13, Exp=ScalarParameter() {defaultValue=1, group=HoneyDew, parameterName=TransitionAmntPow-Hue})
  n19 = MaterialFunctionCall(Hue Shift Percentage=Multiply(A=@n18, B=ScalarParameter() {defaultValue=-0.25, group=HoneyDew, parameterName=Hue}), Texture=@n1.BaseColor) {materialFunction=&HueShift}
  n20 = Multiply(A=@n18, B=VectorParameter() {group=HoneyDew, parameterName=ColorAdd}.RGB)
  FunctionOutput(None=SetMaterialAttributes!(
    MaterialAttributes=@n1.MaterialAttributes
    WorldPositionOffset=Add(A=@n1.WorldPositionOffset, B=@n9)
    OpacityMask=SphereMask(A=@n10, B=@n12, Radius=@n16, Hardness=@n17)
    BaseColor=Add(A=@n19.Result, B=@n20)
   ) {attributeSetTypes [#1, #3, #2]})
 ]}
