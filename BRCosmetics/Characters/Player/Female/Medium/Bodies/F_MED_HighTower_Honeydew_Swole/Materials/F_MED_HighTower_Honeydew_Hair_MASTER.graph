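# Master masked, dithered material for the hair of the HighTower Honeydew Swole female character body, built on material attributes.
# The shared character PreFX function feeds optional silk, cloth fuzz and clear coat layers, each behind a static switch and masked by a selectable channel of SkinFX_Mask.
# The result always passes through the Honeydew hair animation function, then an optional HoloFoil layer using pre-skinned position and normal, before the shared PostFX function.
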Material /BRCosmetics/Characters/Player/Female/Medium/Bodies/F_MED_HighTower_Honeydew_Swole/Materials/F_MED_HighTower_Honeydew_Hair_MASTER
graph {
 properties {
  bAutomaticallySetUsageInEditor=false
  bUseMaterialAttributes=true
  bUsedWithInstancedStaticMeshes=true
  bUsedWithMeshDeformer=true
  bUsedWithSkeletalMesh=true
  blendMode=BLEND_Masked
  ditherOpacityMask=true
  pixelDepthOffsetMode=PDOM_Legacy
  usedShadingModels=Default Lit
 }
 nodes [
  n1 = MaterialFunctionCall() {materialFunction=&BaseFN_CharacterFunction_PreFX}
  n2 = VertexInterpolator(VS=PreSkinnedNormal())
  n3 = TextureSampleParameter2D() {parameterName=SkinFX_Mask}
  n4 = StaticComponentMaskParameter(None=@n3.RGBA) {defaultR=true, group=Skin FX, parameterName=Clear Coat Channel}
  n5 = StaticSwitchParameter(
   True=MaterialFunctionCall!(Alpha=ChannelMaskParameter(None=@n3.RGBA) {group=Skin FX, parameterName=Silk Channel}.R, Material=@n1.Result) {materialFunction=/Game/Weapons/WeaponSkins/MaterialFunction/MF_Silk}.Result
   False=@n1.Result
  ) {group=Skin FX, parameterName=Use Silk}
  n6 = StaticSwitchParameter(
   True=MaterialFunctionCall!(Alpha=StaticComponentMaskParameter(None=@n3.RGBA) {defaultR=true, group=Skin FX, parameterName=Cloth Channel}, Gmap=@n1.Gmap, Material=@n5) {materialFunction=/Game/Weapons/WeaponSkins/MaterialFunction/ClothFuzz}.Result
   False=@n5
  ) {group=Skin FX, parameterName=Use Cloth fuzz}
  n7 = MaterialFunctionCall!(
   GMap=@n1.Gmap
   PreskinnedPosition=@n1.Gmap
   Material=StaticSwitchParameter(True=MaterialFunctionCall!(Alpha=@n4, In=@n6) {materialFunction=/Game/Weapons/WeaponSkins/MaterialFunction/Clearcoat}.Result, False=@n6) {group=Skin FX, parameterName=Use Clear Coat}
  ) {materialFunction=&/MF_HighTower_Honeydew_HairAnimation}
  MaterialFunctionCall(
   AO=@n1.AO
   Distance Blend=@n1.AO
   Fresnel=@n1.AO
   Gmap=@n1.AO
   Material=StaticSwitchParameter(
    True=MaterialFunctionCall!(PreskinnedLocalPosition=@n1.Pre skinned Local Position, PreskinnedLocalNormal=@n2.PS, Material=@n7.Result) {materialFunction=/Game/Characters/Player/Female/Medium/Bodies/F_MED_HighTower_Squash/Materials/MF_HoloFoil}.Result
    False=@n7.Result
   ) {parameterName=BUseHoloFoil}
  ) {materialFunction=&BaseFN_CharacterFunction_PostFX}
 ]
}
