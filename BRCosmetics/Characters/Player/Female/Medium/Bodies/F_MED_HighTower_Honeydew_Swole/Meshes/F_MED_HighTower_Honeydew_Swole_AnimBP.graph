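# Post-process-style Anim Blueprint for the HighTower Honeydew Swole body cosmetic, adding muscle deformation to a pose copied from the attached parent mesh.
# The main pose chains constraints and additive bone edits on thigh, elbow, bicep, tricep and deltoid helpers, with alphas range-mapped from joint rotation variables.
# In the front end, a lobby animation evaluated at MusclesPosition instead layers onto arm muscle bones by EmoteAlpha.
# Driving variables are set outside this empty EventGraph.
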AnimBlueprint /BRCosmetics/Characters/Player/Female/Medium/Bodies/F_MED_HighTower_Honeydew_Swole/Meshes/F_MED_HighTower_Honeydew_Swole_AnimBP
target_skeleton=&/F_MED_HighTower_Honeydew_Swole_Skeleton
graphs [
 AnimGraph {nodes [
   Output Pose(Result=Blend Poses by bool(BlendPose_0=Use cached pose 'frontEndPose'(), BlendPose_1=Use cached pose 'mainPose'(), bActiveValue=Get bIsInFrontEnd()) {blendType=Linear}) {layerGroup=None}
   n1 = "Constraint - thigh_volume_l (R)"(ComponentPose=Local To Component(LocalPose=Copy Pose From Mesh() {bUseAttachedParent=true}), Alpha=0.5, ConstraintWeights_0=1) {
    alphaBoolBlend {blendOption=Linear}
    boneToModify {boneName=thigh_volume_l}
    constraintSetup [{transformType=Rotation, targetBone {boneName=thigh_l}}]
    constraintWeights [1]
   }
   n2 = "Constraint - thigh_volume_r (R)"(ComponentPose=@n1, Alpha=0.5, ConstraintWeights_0=1) {
    alphaBoolBlend {blendOption=Linear}
    boneToModify {boneName=thigh_volume_r}
    constraintSetup [{transformType=Rotation, targetBone {boneName=thigh_r}}]
    constraintWeights [1]
   }
   n3 = "Transform (Modify) Bone - Bone: thigh_volume_l"(ComponentPose=@n2, Alpha=Get ThighRotationLeft() -> (ThighRotationLeft:float), Translation="-1,-2,1", Rotation="0,0,0", Scale="1,1,1") {
    translationMode=BMM_Additive
    translationSpace=BCS_BoneSpace
    alphaBoolBlend {blendOption=Linear}
    alphaScaleBiasClamp {bMapRange=true, inRange {max=0, min=-100}, outRange {max=0, min=1}}
    boneToModify {boneName=thigh_volume_l}
    translation=<-1 -2 1>
   }
   n4 = "Transform (Modify) Bone - Bone: thigh_volume_r"(ComponentPose=@n3, Alpha=Get ThighRotationRight() -> (ThighRotationRight:float), Translation="1,2,-1", Rotation="0,0,0", Scale="1,1,1") {
    translationMode=BMM_Additive
    translationSpace=BCS_BoneSpace
    alphaBoolBlend {blendOption=Linear}
    alphaScaleBiasClamp {bMapRange=true, inRange {max=0, min=-100}, outRange {max=0, min=1}}
    boneToModify {boneName=thigh_volume_r}
    translation=<1 2 -1>
   }
   n5 = "Constraint - dyn_elbow_base_l (PP)"!(ComponentPose=@n4, Alpha=1, ConstraintWeights_0=0.7, ConstraintWeights_1:float=0.3) -> (Pose) {
    alphaBoolBlend {blendOption=Linear}
    boneToModify {boneName=dyn_elbow_base_l}
    constraintSetup [{targetBone {boneName=upperarm_l}}, {targetBone {boneName=lowerarm_l}}]
    constraintWeights [1,1]
   }
   n6 = "Transform (Modify) Bone - Bone: dyn_elbow_l"(ComponentPose=@n5, Alpha=Get ElbowRotationLeft() -> (ElbowRotationLeft:float), Translation="0,-2.5,0", Rotation="0,0,0", Scale="1,1,1") {
    translationMode=BMM_Additive
    translationSpace=BCS_BoneSpace
    alphaBoolBlend {blendOption=Linear}
    alphaScaleBiasClamp {bMapRange=true, inRange {max=100}}
    boneToModify {boneName=dyn_elbow_l}
    translation {y=-2.5}
   }
   n7 = "Constraint - dyn_elbow_base_r (PP)"!(ComponentPose=@n6, Alpha=1, ConstraintWeights_0=0.7, ConstraintWeights_1:float=0.3) -> (Pose) {
    alphaBoolBlend {blendOption=Linear}
    boneToModify {boneName=dyn_elbow_base_r}
    constraintSetup [{targetBone {boneName=upperarm_r}}, {targetBone {boneName=lowerarm_r}}]
    constraintWeights [1,1]
   }
   n8 = "Transform (Modify) Bone - Bone: dyn_elbow_r"(ComponentPose=@n7, Alpha=Get ElbowRotationRight() -> (ElbowRotationRight:float), Translation="0,2.5,0", Rotation="0,0,0", Scale="1,1,1") {
    translationMode=BMM_Additive
    translationSpace=BCS_BoneSpace
    alphaBoolBlend {blendOption=Linear}
    alphaScaleBiasClamp {bMapRange=true, inRange {max=100}}
    boneToModify {boneName=dyn_elbow_r}
    translation {y=2.5}
   }
   n9 = "Transform (Modify) Bone - Bone: dyn_bicep_l"(ComponentPose=@n8, Alpha=Get ElbowRotationLeft(), Translation="-2,2,0", Rotation="0,0,0", Scale="1,1,1.5") {
    rotationMode=BMM_Additive
    rotationSpace=BCS_BoneSpace
    scaleMode=BMM_Additive
    scaleSpace=BCS_BoneSpace
    translationMode=BMM_Additive
    translationSpace=BCS_BoneSpace
    alphaBoolBlend {blendOption=Linear}
    alphaScaleBiasClamp {bMapRange=true, inRange {max=100}}
    boneToModify {boneName=dyn_bicep_l}
    scale {z=1.5}
    translation=<-2 2>
   }
   n10 = "Transform (Modify) Bone - Bone: dyn_bicep_r"(ComponentPose=@n9, Alpha=Get ElbowRotationRight(), Translation="2,-2,0", Rotation="0,0,0", Scale="1,1,1.5") {
    rotationMode=BMM_Additive
    rotationSpace=BCS_BoneSpace
    scaleMode=BMM_Additive
    scaleSpace=BCS_BoneSpace
    translationMode=BMM_Additive
    translationSpace=BCS_BoneSpace
    alphaBoolBlend {blendOption=Linear}
    alphaScaleBiasClamp {bMapRange=true, inRange {max=100}}
    boneToModify {boneName=dyn_bicep_r}
    scale {z=1.5}
    translation=<2 -2>
   }
   n11 = "Transform (Modify) Bone - Bone: dyn_tricep_l"(ComponentPose=@n10, Alpha=Get ElbowRotationLeft(), Translation="3,1,0", Rotation="0,0,0", Scale="1,1,1") {
    translationMode=BMM_Additive
    translationSpace=BCS_BoneSpace
    alphaBoolBlend {blendOption=Linear}
    alphaScaleBiasClamp {bMapRange=true, inRange {max=100}}
    boneToModify {boneName=dyn_tricep_l}
    translation=<3 1>
   }
   n12 = "Transform (Modify) Bone - Bone: dyn_tricep_r"(ComponentPose=@n11, Alpha=Get ElbowRotationRight(), Translation="-3,-1,0", Rotation="0,0,0", Scale="1,1,1") {
    translationMode=BMM_Additive
    translationSpace=BCS_BoneSpace
    alphaBoolBlend {blendOption=Linear}
    alphaScaleBiasClamp {bMapRange=true, inRange {max=100}}
    boneToModify {boneName=dyn_tricep_r}
    translation=<-3 -1>
   }
   n13 = "Transform (Modify) Bone - Bone: dyn_deltoid_l"(
    ComponentPose=@n12
    Alpha=Get UpperarmRotationLeft() -> (UpperarmRotationLeft:float)
    Translation="0,0,1"
    Rotation="0,0,0"
    Scale="1.1,1.2,1"
   ) {
    rotationMode=BMM_Additive
    rotationSpace=BCS_BoneSpace
    scaleMode=BMM_Additive
    scaleSpace=BCS_BoneSpace
    translationMode=BMM_Additive
    translationSpace=BCS_BoneSpace
    alphaBoolBlend {blendOption=Linear}
    alphaScaleBiasClamp {bMapRange=true, inRange {max=40}}
    boneToModify {boneName=dyn_deltoid_l}
    scale=<1.1 1.2>
    translation {z=1}
   }
   n14 = "Transform (Modify) Bone - Bone: dyn_deltoid_r"(
    ComponentPose=@n13
    Alpha=Get UpperarmRotationRight() -> (UpperarmRotationRight:float)
    Translation="0,0,-1"
    Rotation="0,0,0"
    Scale="1.1,1.2,1"
   ) {
    rotationMode=BMM_Additive
    rotationSpace=BCS_BoneSpace
    scaleMode=BMM_Additive
    scaleSpace=BCS_BoneSpace
    translationMode=BMM_Additive
    translationSpace=BCS_BoneSpace
    alphaBoolBlend {blendOption=Linear}
    alphaScaleBiasClamp {bMapRange=true, inRange {max=40}}
    boneToModify {boneName=dyn_deltoid_r}
    scale=<1.1 1.2>
    translation {z=-1}
   }
   Save cached pose 'mainPose'(Pose=Component To Local(ComponentPose=@n14)) {cacheName=mainPose}
   n15 = Layered blend per bone(
    BasePose=Copy Pose From Mesh() {bUseAttachedParent=true}
    BlendPoses_0="Sequence Evaluator (by time) 'Hightower_Honeydew_BP_Idle_Elbow'"!(ExplicitTime:float=Get MusclesPosition() -> (MusclesPosition:float)) -> (Pose:Pose Link Structure) {
     is=title AnimGraphNode_SequenceEvaluator
     sequence=/BRCosmetics/Animation/Game/MainPlayer/Menu/BR/S14_BP_Lobby/Hightower_Honeydew_BP_Idle_Elbow
    }
    BlendWeights_0=Get EmoteAlpha() -> (EmoteAlpha:float)
   ) {layerSetup [{branchFilters [
       {boneName=dyn_elbow_l}
       {boneName=dyn_bicep_l}
       {boneName=dyn_tricep_l}
       {boneName=dyn_deltoid_l}
       {boneName=dyn_elbow_r}
       {boneName=dyn_bicep_r}
       {boneName=dyn_tricep_r}
       {boneName=dyn_deltoid_r}
      ]}]}
   Save cached pose 'frontEndPose'(Pose=@n15) {cacheName=frontEndPose}
  ]}
 EventGraph {links [], nodes []}
]
variables []
